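# Translucent unlit Niagara mesh-particle material for airflow streaks in the S4100 front-end effects.
# SplineThicken sets mesh width, with the dynamic WidthBase parameter driving both base and tip.
# A clamped soft-flair texture offset along V by the Scroll parameter sweeps through swirly smoke breakup.
# Opacity math adds U-edge falloff, a fade near V=1 and particle alpha, while particle color feeds MF_FX_EmissiveAdjustments.
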Material /Game/Effects/Fort_Effects/FrontEnd/S4100/M_FE_4100_AirFlow
graph {
 properties {
  bAutomaticallySetUsageInEditor=false
  bUsedWithNiagaraMeshParticles=true
  blendMode=BLEND_Translucent
  shadingModel=MSM_Unlit
  usedShadingModels=Unlit
 }
 nodes [
  Constant3Vector() {constant=<rgba 0.7622 0.7622 0.7622 0>}
  Constant() {r=10}
  n1 = DynamicParameter() {defaultValue=<rgba 0.5 1 1 1>, paramNames [Scroll,WidthBase,WidthTip,Param4]}
  MaterialFunctionCall(WidthBase=@n1.WidthBase, WidthTip=@n1.WidthBase, Expand U or V UV Channel=StaticBool() {value=true}) {materialFunction=&SplineThicken}
  n2 = ParticleColor()
  MaterialFunctionCall(Emissive=@n2.RGB) {materialFunction=&MF_FX_EmissiveAdjustments}
  n3 = TextureSample(UVs=Multiply(A=TextureCoordinate()) {constB=0.5}) {texture=/Game/Effects/Fort_Effects/Textures/T_Swirly_Smoke}
  n4 = TextureCoordinate()
  n5 = AppendVector(A=Constant(), B=@n1.Scroll)
  n6 = TextureCoordinate()
  n7 = Multiply(A=Subtract(A=ComponentMask(None=@n6) {r=true}) {constB=0.5}) {constB=2}
  n8 = Subtract(
   A=TextureSample(UVs=Add(A=@n4, B=@n5)) {samplerSource=SSM_Clamp_WorldGroupSettings, texture=/Game/Effects/Fort_Effects/Textures/T_New_Soft_Flair}.R
   B=Power(Base=Abs(None=@n7)) {constExponent=6}
  )
  n9 = ComponentMask(None=@n6) {g=true}
  Multiply(
   A=Multiply(A=Power(Base=@n3.R) {constExponent=1.4}, B=Multiply(A=@n2.A, B=@n8))
   B=SmoothStep(Value=OneMinus(None=@n9)) {constMax=0.4, constMin=0.01}
  )
 ]
}
