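# Niagara system for the S4100 frontend scene's ambient effects: drifting air lines, ground snow, birds, glows and a lens flare.
# Mesh air lines and snow fade over life, while two GPU bird emitters drive flapping through material parameters, only one using wind and vortex forces.
# Persistent burst sprites form glow, flare and dust, mostly brightened by a system Timeline value.
# Four array-placed light glows vary in alpha and shift position with the IsZoomedOut flag.
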NiagaraSystem /Game/Effects/Fort_Effects/FrontEnd/S4100/NS_4100_Frontend
dependencies {
 usedDataInterfaces [
  &NiagaraDataInterfaceCurve
  &NiagaraDataInterfaceArrayColor
  &NiagaraDataInterfaceArrayFloat3
  &NiagaraDataInterfaceArrayFloat
  &NiagaraDataInterfaceMeshRendererInfo
  &NiagaraDataInterfaceColorCurve
  &NiagaraDataInterfaceParticleRead
  &NiagaraDataInterfaceDebugDraw
 ]
 usedDynamicInputs [&SimulationPosition, &SetBoolByFloatComparison]
 usedModules [
  &self:SystemSpawnScript.SystemScriptSource.SystemScriptGraph.NiagaraNodeAssignment_1.SetVariables_#1
  &SystemState
  /Niagara/Noise/Waveform
  &self:SystemSpawnScript.SystemScriptSource.SystemScriptGraph.NiagaraNodeAssignment_0.SetVariables_#2
  &Timeline
 ]
}
emitters {
 AirLines {
  emitterUpdateScript [
   &EmitterState()
   &SpawnRate(Spawn Probability=0.7, SpawnRate=8, Use Spawn Probability=true)
  ]
  particleSpawnScript [
   &InitializeParticle(
    Color Maximum=<rgba 1 0.5864 0.8611 0.22>
    Color Mode=::NewEnumerator2
    Lifetime Max=5.5
    Lifetime Min=3.5
    Lifetime Mode=::NewEnumerator1
    Mesh Renderer Array Visibility Mode=::NewEnumerator3
    Mesh Renderer Info=DataInterface() {propertyValues {MeshRenderer=&self:AirLines_0.NiagaraMeshRendererProperties_0}}
    Mesh Scale Max=<3.5 2 2>
    Mesh Scale Mode=::NewEnumerator2
    Position Offset=<-250 0 100>
    UsePositionOffset=true
    Write Material Random=true
   )
   &ShapeLocation(
    Box Midpoint=<0.7773 0.5 0.324>
    Box Size=<0 1220 250>
    MaskSpawnGroup=false
    Override Seed=false
    Shape Primitive=::NewEnumerator2
    Surface Only=false
    UseOwnerScale=true
   )
   &AddVelocity(Velocity=<0 -10 0>)
   &InitialMeshOrientation(Mesh Orientation Mode=::NewEnumerator2, Orientation Vector=<0 -1 0>, Use Rotation Vector=false) {enabled=false}
  ]
  particleUpdateScript [
   &ParticleState()
   &ScaleColor(
    Scale Alpha=&FloatFromCurve(
     FloatCurve=DataInterface() {propertyValues {Curve {keys [
         {arriveTangent=4.322, interpMode=RCIM_Cubic, leaveTangent=4.322, tangentMode=RCTM_User, time=0.0045, value=0.2314}
         {arriveTangent=1.5704, tangentMode=RCTM_User, time=0.3117, value=1.1925}
         <at 1 0>
        ]}}}
     Scale Curve=&UniformAOrBFloat(A=1.2, B=0.8, Distribution Weight=0.8, Evaluation Type=::NewEnumerator0, Recalculate Random Each Loop=true, Override Seed=false)
    )
    ScaleA=true
    ScaleRGB=true
   )
   &SolveForcesAndVelocity(Clamp Velocity=false, Limit Acceleration=false)
   &DynamicMaterialParameters(
    Index 0 Param 1=&FloatFromCurve(FloatCurve=DataInterface() {propertyValues {Curve {keys [
         {arriveTangent=2.1823, interpMode=RCIM_Cubic, tangentMode=RCTM_Break, time=0, value=-1}
         {arriveTangent=2.1823, tangentMode=RCTM_User, time=1, value=1}
        ]}}})
    Index 0 Param 2=&FloatFromCurve(
     FloatCurve=DataInterface() {propertyValues {Curve {
        defaultValue=1
        keys [
         {tangentMode=RCTM_User, time=-0.0011, value=0.0925}
         <at 0.1599 1.0282>
         {arriveTangent=-0.0405, interpMode=RCIM_Cubic, leaveTangent=-0.0405, tangentMode=RCTM_User, time=0.5844, value=1.0519}
         {arriveTangent=-5.8573, interpMode=RCIM_Cubic, leaveTangent=-5.8573, tangentMode=RCTM_User, time=1, value=0}
        ]
       }}}
     Scale Curve=&RandomRangeFloat(Minimum=25, Maximum=40)
    )
    Index 0 Param 3=&FloatFromCurve(
     FloatCurve=DataInterface() {propertyValues {Curve {keys [
         <at 0.0108 0.2876>
         {interpMode=RCIM_Cubic, tangentMode=RCTM_Break, time=0.2076, value=1.0455}
         {arriveTangent=-0.5497, tangentMode=RCTM_User, time=1, value=0}
        ]}}}
     Scale Curve=25
    )
    Module.Index0Float0Write=true
    Module.Index0Float1Write=true
    Module.Index0Float2Write=true
    Module.Index0Float3Write=true
   ) {module=DynamicMaterialParameters001}
  ]
  rendererClasses [&NiagaraMeshRendererProperties]
  renderers [
   {rendererClass=&NiagaraMeshRendererProperties, rendererIndex=0}
   {rendererClass=&NiagaraMeshRendererProperties, rendererIndex=1}
  ]
 }
 AirLines-InFront {
  emitterUpdateScript [
   &EmitterState()
   &SpawnRate(Spawn Probability=0.2, SpawnRate=3, Use Spawn Probability=true)
  ]
  particleSpawnScript [
   &InitializeParticle(
    Color Maximum=<rgba 1 0.6872 0.8053 0.3>
    Color Minimum=<rgba 0.8729 0.7197 1 0.3>
    Color Mode=::NewEnumerator2
    Lifetime Max=5.5
    Lifetime Min=3.5
    Lifetime Mode=::NewEnumerator1
    Mesh Renderer Array Visibility Mode=::NewEnumerator3
    Mesh Renderer Info=DataInterface() {propertyValues {MeshRenderer=&self:AirLines-InFront_0.NiagaraMeshRendererProperties_0}}
    Mesh Scale Max=<3.5 2 2>
    Mesh Scale Mode=::NewEnumerator2
    Position Offset=<-250 925 0>
    UsePositionOffset=true
    Write Material Random=true
   )
   &ShapeLocation(
    Box Midpoint=<0.7773 0.5 0.324>
    Box Size=<0 400 120>
    MaskSpawnGroup=false
    Override Seed=false
    Shape Primitive=::NewEnumerator2
    Surface Only=false
    UseOwnerScale=true
   )
   &AddVelocity(Velocity=<0 -10 0>)
   &InitialMeshOrientation(Mesh Orientation Mode=::NewEnumerator2, Orientation Vector=<0 -1 0>, Use Rotation Vector=false) {enabled=false}
  ]
  particleUpdateScript [
   &ParticleState()
   &ScaleColor(
    Scale Alpha=&FloatFromCurve(
     FloatCurve=DataInterface() {propertyValues {Curve {keys [
         {arriveTangent=4.322, interpMode=RCIM_Cubic, leaveTangent=4.322, tangentMode=RCTM_User, time=0.0045, value=0.2314}
         {arriveTangent=1.5704, tangentMode=RCTM_User, time=0.3117, value=1.1925}
         <at 1 0>
        ]}}}
     Scale Curve=&UniformAOrBFloat(A=1.2, B=0.8, Distribution Weight=0.8, Evaluation Type=::NewEnumerator0, Recalculate Random Each Loop=true, Override Seed=false)
    )
    ScaleA=true
    ScaleRGB=true
   )
   &SolveForcesAndVelocity(Clamp Velocity=false, Limit Acceleration=false)
   &DynamicMaterialParameters(
    Index 0 Param 1=&FloatFromCurve(FloatCurve=DataInterface() {propertyValues {Curve {keys [
         {arriveTangent=2.1823, interpMode=RCIM_Cubic, tangentMode=RCTM_Break, time=0, value=-1}
         {arriveTangent=2.1823, tangentMode=RCTM_User, time=1, value=1}
        ]}}})
    Index 0 Param 2=&FloatFromCurve(
     FloatCurve=DataInterface() {propertyValues {Curve {
        defaultValue=1
        keys [
         {tangentMode=RCTM_User, time=-0.0011, value=0.0925}
         <at 0.1599 1.0282>
         {arriveTangent=-0.0405, interpMode=RCIM_Cubic, leaveTangent=-0.0405, tangentMode=RCTM_User, time=0.5844, value=1.0519}
         {arriveTangent=-5.8573, interpMode=RCIM_Cubic, leaveTangent=-5.8573, tangentMode=RCTM_User, time=1, value=0}
        ]
       }}}
     Scale Curve=&RandomRangeFloat(Minimum=12, Maximum=20)
    )
    Index 0 Param 3=&FloatFromCurve(
     FloatCurve=DataInterface() {propertyValues {Curve {keys [
         <at 0.0108 0.2876>
         {interpMode=RCIM_Cubic, tangentMode=RCTM_Break, time=0.2076, value=1.0455}
         {arriveTangent=-0.5497, tangentMode=RCTM_User, time=1, value=0}
        ]}}}
     Scale Curve=12
    )
    Module.Index0Float0Write=true
    Module.Index0Float1Write=true
    Module.Index0Float2Write=true
    Module.Index0Float3Write=true
   ) {module=DynamicMaterialParameters001}
  ]
  rendererClasses [&NiagaraMeshRendererProperties]
  renderers [
   {rendererClass=&NiagaraMeshRendererProperties, rendererIndex=0}
   {rendererClass=&NiagaraMeshRendererProperties, rendererIndex=1}
  ]
 }
 BigGlow {
  emitterUpdateScript [
   &EmitterState(Life Cycle Mode=::NewEnumerator1, Loop Behavior=::NewEnumerator1, Loop Duration Mode=::NewEnumerator1, UseLoopDelay=false)
   &SpawnBurst_Instantaneous(Use Spawn Probability=false, UseLoopCountLimit=false)
  ]
  particleSpawnScript [
   &InitializeParticle(
    Color=<rgba 1 0.3526 0.6329 0.85>
    Lifetime=5
    Position Offset=<710 0 150>
    Sprite Size Mode=::NewEnumerator3
    Uniform Sprite Size=2450
    UsePositionOffset=true
    Write Material Random=true
   )
   &CameraOffset(Camera Offset Amount=100)
  ]
  particleUpdateScript [
   &ParticleState(Kill Particles When Lifetime Has Elapsed=false)
   &ScaleColor(Scale Alpha=&Lerp_Float(A=0.7, B=1, Alpha=$System.Timeline.Time), ScaleA=true, ScaleRGB=true)
  ]
  rendererClasses [&NiagaraSpriteRendererProperties]
  renderers [{rendererClass=&NiagaraSpriteRendererProperties}]
 }
 Birds {
  simTarget=GPUComputeSim
  emitterUpdateScript [
   &EmitterState()
   &SpawnBurst_Instantaneous(
    Spawn Count=&RandomRangeInt(Minimum=4, Maximum=8)
    Spawn Probability=0.05
    Use Spawn Probability=true
    UseLoopCountLimit=false
   )
   &SpawnRate(Spawn Probability=0.35, SpawnRate=3, Use Spawn Probability=true)
  ]
  particleSpawnScript [
   &FN_SetSpawnGroup()
   &InitializeParticle(
    Color=<rgba 2 2 2 1>
    Lifetime=&RandomRangeFloat(Minimum=4, Maximum=10)
    Mesh Scale Mode=::NewEnumerator1
    Mesh Uniform Scale Max=0.21
    Mesh Uniform Scale Min=0.1
    Sprite Size Mode=::NewEnumerator3
    Uniform Sprite Size=500
    UsePositionOffset=false
    Write Material Random=true
   )
   &SphereLocation(
    MaskSpawnGroup=false
    Non Uniform Scale=<3 1 0.6>
    Offset=<-150 0 50>
    Override Seed=false
    Sphere Radius=120
    Surface Only=false
    UseNonUniformScale=true
    UseOffset=true
    UseSphereOrientation=false
   )
   &AddVelocity(Velocity=&RandomRangeVector(Minimum=<-65 -5 150>, Maximum=<65 -8 250>))
  ]
  particleUpdateScript [
   &ParticleState()
   &ScaleColor(
    Linear Color Curve=DataInterface() {propertyValues {
      AlphaCurve {keys [<at 0 1>]}
      BlueCurve {keys [<at 0.3018 1>,<at 0.5878 0.6939>]}
      GreenCurve {keys [<at 0.3018 1>,<at 0.5878 0.5029>]}
      RedCurve {keys [<at 0.3018 1>,<at 0.5878 1>]}
     }}
    Scale Mode=2
   )
   &self:PositionReader(
    IndexID:NiagaraInt32=&RandomRangeInt(Minimum=0, Maximum=1)
    New Particle Attribute Reader=DataInterface() {propertyValues {EmitterBinding {bindingMode=Other, emitterName=BirdFlocker}}}
   ) {enabled=false, module=PositionReader}
   /SetVariables(Particles.RandomDirection:Vector3f=&Lerp_Vector(
     A=&MakeVector(
      X=/Niagara/DynamicInputs/Helpers/Waveform(
       Global Phase=&Add_Float(A=$Particles.Age, B=&RandomRangeFloat(Minimum=0, Maximum=5))
       "Waveform [1]"=::NewEnumerator8
       "[1] Frequency"=1
       "[1] Amplitude Min/Max"=<-1 1>
       "[2] Phase"=$Emitter.Age
       "[3] Phase"=$Emitter.Age
      )
      Y=&Waveform(
       Global Phase=&Add_Float(A=$Particles.Age, B=&RandomRangeFloat(Minimum=0, Maximum=5))
       "Waveform [1]"=::NewEnumerator8
       "[1] Frequency"=1
       "[1] Amplitude Min/Max"=<-1 1>
       "[2] Phase"=$Emitter.Age
       "[3] Phase"=$Emitter.Age
      )
      Z=0
     )
     B=<1 0 0>
     Alpha=&Waveform(
      Global Phase=&Add_Float(A=$Particles.Age, B=&RandomRangeFloat(Minimum=0, Maximum=5))
      Global Clamp Min=0.4
      "Waveform [1]"=::NewEnumerator8
      "[1] Frequency"=1
      "[1] Amplitude Min/Max"=<0 1>
      "[2] Phase"=$Emitter.Age
      "[3] Phase"=$Emitter.Age
     )
    )) {isSetParameters=true, module=SetVariables_#3}
   &InterpolateOverTime(
    Rate Of Change=0.5
    Target Vector Value:Vector3f=&NormalizeVector(Vector To Normalize=$Particles.RandomDirection, Safe Normalize=true, Safe Fallback Threshold=0, Safe Fallback Vector=<0 0 0.001>)
    Variable Type=::NewEnumerator2
   ) {module=InterpolateOverTime001}
   &WindForce(
    Air Resistance=0.1
    Wind Speed=&TransformVector(Vector=<-10 0 1>, Destination Space=::Simulation)
    Wind Speed Scale=&RandomRangeFloat(Minimum=4, Maximum=5)
   )
   &VortexForce(
    Origin Pull Amount=15
    UseOriginOffset=true
    Vortex Axis=&Lerp_Vector(
     A=<-0.15 0.15 0>
     B=<0.25 0 0.25>
     Alpha=&FloatFromCurve(
      FloatCurve=DataInterface() {propertyValues {Curve {keys [
          {interpMode=RCIM_Cubic, tangentMode=RCTM_User, time=0, value=0}
          {interpMode=RCIM_Cubic, tangentMode=RCTM_User, time=0.5, value=1}
          {tangentMode=RCTM_User, time=1, value=0}
         ]}}}
      CurveIndex=$System.NormalizedLoopAge
     )
    )
    Vortex Force Amount=50
    Vortex Origin:Vector3f=$Particles.FishMoverPosition
    Vortex Origin Offset:Vector3f=<0 0 100>
   )
   &SolveForcesAndVelocity(Acceleration Limit=50, Clamp Velocity=true, Limit Acceleration=true, Speed Limit=40)
   /SetVariables(Particles.FlapTarget:NiagaraFloat=/Niagara/DynamicInputs/Integer/V2/Round_Float(Float:NiagaraFloat=&Waveform(
      "Waveform [1]"=::NewEnumerator8
      "[1] Phase"=&Add_Float(A=$System.Age, B=&RandomRangeFloat(Minimum=0, Maximum=2))
      "[1] Frequency"=&RandomRangeFloat(Minimum=0.4, Maximum=1)
      "[1] Amplitude Min/Max"=<0 1>
      "[2] Average"=false
      "[2] Input Position"=&RandomRangeFloat(Minimum=0, Maximum=1)
      "[2] Phase"=$Particles.Age
      "[3] Phase"=$Emitter.Age
     ))) {enabled=false, isSetParameters=true, module=SetVariables_#4}
   /SetVariables(Particles.FlapTarget=1) {isSetParameters=true, module=SetVariables_#5}
   &InterpolateOverTime(
    Rate Of Change:NiagaraFloat=&Lerp_Float(A=1, B=1, Alpha=$Particles.FlapTarget)
    Target Float Value:NiagaraFloat=$Particles.FlapTarget
   )
   /SetVariables(Particles.FlapSmoothed:NiagaraFloat=$Particles.InterpolateOverTime.InterpolateOverTime_Float.Moving Average) {isSetParameters=true, module=SetVariables_#6}
   /Niagara/Modules/Update/Utility/IncrementOverTime(Rate of Change:NiagaraFloat=&Multiply_Float(
     A=&Add_Float(A=0.15, B=$Particles.FlapSmoothed)
     B=&RemapRange(
      Input Value To Remap=&VectorLength(Vector=$Particles.Velocity)
      Input Min=0
      Input Max=5000
      Output Min=1
      Output Max=5
     )
    ))
   &DynamicMaterialParameters(
    Index 0 Param 1=&Multiply_Float(A=$Particles.FlapSmoothed, B=1)
    Index 0 Param 2=&Multiply_Float(A=$Particles.IncrementOverTime.Counter, B=1)
    Module.Param0WriteEnabled=true
    Module.Param1WriteEnabled=true
    Module.Param2WriteEnabled=false
    Module.Param3WriteEnabled=false
   ) {module=DynamicMaterialParameters001}
   &RegionMask(Outer Vector:Vector3f=<0 0 2>, Region Radius=800, UseOffset=false, UseRotation=false, Value Type:ENiagaraRegionMaskValue=::NewEnumerator1)
   /SetVariables(Particles.Scale=$Particles.RegionMask.Vector) {isSetParameters=true, module=SetVariables_#7}
   &KillParticlesInVolume(Invert Volume=true, Sphere Radius=2500)
   /SetVariables(Particles.NewScale:Vector3f=&Multiply_VectorByFloat(
     Vector=$Particles.Scale
     Float=&FloatFromCurve(
      FloatCurve=DataInterface() {propertyValues {Curve {keys [
          {interpMode=RCIM_Cubic, tangentMode=RCTM_User, time=0.3266, value=0.102}
          {interpMode=RCIM_Cubic, tangentMode=RCTM_User, time=0.5025, value=0.9922}
          {interpMode=RCIM_Cubic, tangentMode=RCTM_User, time=0.7043, value=1.0157}
          {tangentMode=RCTM_User, time=1, value=0}
         ]}}}
      Scale Curve=0.1
     )
    )) {isSetParameters=true, module=SetVariables_#8}
  ]
  rendererClasses [&NiagaraMeshRendererProperties, &NiagaraSpriteRendererProperties]
  renderers [
   {rendererClass=&NiagaraMeshRendererProperties, rendererIndex=0}
   {rendererClass=&NiagaraSpriteRendererProperties, rendererIndex=1}
  ]
 }
 Birds001 {
  simTarget=GPUComputeSim
  emitterUpdateScript [
   &EmitterState()
   &SpawnRate(Spawn Probability=0.15, SpawnRate=1.5, Use Spawn Probability=true)
  ]
  particleSpawnScript [
   &FN_SetSpawnGroup()
   &InitializeParticle(
    Color=<rgba 2 2 2 1>
    Lifetime=&RandomRangeFloat(Minimum=13, Maximum=15)
    Mesh Scale Mode=::NewEnumerator1
    Mesh Uniform Scale Max=0.21
    Mesh Uniform Scale Min=0.1
    Sprite Size Mode=::NewEnumerator3
    Uniform Sprite Size=500
    UsePositionOffset=false
    Write Material Random=true
   )
   &SphereLocation(
    MaskSpawnGroup=false
    Non Uniform Scale=<3 1 0.6>
    Offset=<-150 0 50>
    Override Seed=false
    Sphere Radius=120
    Surface Only=false
    UseNonUniformScale=true
    UseOffset=true
    UseSphereOrientation=false
   )
   &AddVelocity(Velocity=&RandomRangeVector(Minimum=<320 -80 220>, Maximum=<450 80 265>))
  ]
  particleUpdateScript [
   &ParticleState()
   &ScaleColor(
    Linear Color Curve=DataInterface() {propertyValues {
      AlphaCurve {keys [<at 0 1>]}
      BlueCurve {keys [<at 0.3018 1>,<at 0.5878 0.6939>]}
      GreenCurve {keys [<at 0.3018 1>,<at 0.5878 0.5029>]}
      RedCurve {keys [<at 0.3018 1>,<at 0.5878 1>]}
     }}
    Scale Mode=2
   )
   &self:PositionReader(
    IndexID:NiagaraInt32=&RandomRangeInt(Minimum=0, Maximum=1)
    New Particle Attribute Reader=DataInterface() {propertyValues {EmitterBinding {bindingMode=Other, emitterName=BirdFlocker}}}
   ) {enabled=false, module=PositionReader}
   /SetVariables(Particles.RandomDirection:Vector3f=&Lerp_Vector(
     A=&MakeVector(
      X=&Waveform(
       Global Phase=&Add_Float(A=$Particles.Age, B=&RandomRangeFloat(Minimum=0, Maximum=5))
       "Waveform [1]"=::NewEnumerator8
       "[1] Frequency"=1
       "[1] Amplitude Min/Max"=<-1 1>
       "[2] Phase"=$Emitter.Age
       "[3] Phase"=$Emitter.Age
      )
      Y=&Waveform(
       Global Phase=&Add_Float(A=$Particles.Age, B=&RandomRangeFloat(Minimum=0, Maximum=5))
       "Waveform [1]"=::NewEnumerator8
       "[1] Frequency"=1
       "[1] Amplitude Min/Max"=<-1 1>
       "[2] Phase"=$Emitter.Age
       "[3] Phase"=$Emitter.Age
      )
      Z=0
     )
     B=<1 0 0>
     Alpha=&Waveform(
      Global Phase=&Add_Float(A=$Particles.Age, B=&RandomRangeFloat(Minimum=0, Maximum=5))
      Global Clamp Min=0.4
      "Waveform [1]"=::NewEnumerator8
      "[1] Frequency"=1
      "[1] Amplitude Min/Max"=<0 1>
      "[2] Phase"=$Emitter.Age
      "[3] Phase"=$Emitter.Age
     )
    )) {isSetParameters=true, module=SetVariables_#3}
   &InterpolateOverTime(
    Rate Of Change=0.5
    Target Vector Value:Vector3f=&NormalizeVector(Vector To Normalize=$Particles.RandomDirection, Safe Normalize=true, Safe Fallback Threshold=0, Safe Fallback Vector=<0 0 0.001>)
    Variable Type=::NewEnumerator2
   ) {module=InterpolateOverTime001}
   &WindForce(
    Air Resistance=0.1
    Wind Speed=&TransformVector(Vector=<12 -10 5>, Destination Space=::Simulation)
    Wind Speed Scale=&RandomRangeFloat(Minimum=4, Maximum=5)
   ) {enabled=false}
   &VortexForce(
    Origin Pull Amount=15
    UseOriginOffset=true
    Vortex Axis=&Lerp_Vector(
     A=<0 -0.15 0>
     B=<0 -0.3 0.25>
     Alpha=&FloatFromCurve(
      FloatCurve=DataInterface() {propertyValues {Curve {keys [
          {interpMode=RCIM_Cubic, tangentMode=RCTM_User, time=0, value=0}
          {interpMode=RCIM_Cubic, tangentMode=RCTM_User, time=0.5, value=1}
          {tangentMode=RCTM_User, time=1, value=0}
         ]}}}
      CurveIndex=$System.NormalizedLoopAge
     )
    )
    Vortex Force Amount=50
    Vortex Origin:Vector3f=$Particles.FishMoverPosition
    Vortex Origin Offset:Vector3f=<0 0 100>
   ) {enabled=false}
   &SolveForcesAndVelocity(Acceleration Limit=50, Clamp Velocity=true, Limit Acceleration=true, Speed Limit=40)
   /SetVariables(Particles.FlapTarget:NiagaraFloat=&Round_Float(Float:NiagaraFloat=&Waveform(
      "Waveform [1]"=::NewEnumerator8
      "[1] Phase"=&Add_Float(A=$System.Age, B=&RandomRangeFloat(Minimum=0, Maximum=2))
      "[1] Frequency"=&RandomRangeFloat(Minimum=0.4, Maximum=1)
      "[1] Amplitude Min/Max"=<0 1>
      "[2] Average"=false
      "[2] Input Position"=&RandomRangeFloat(Minimum=0, Maximum=1)
      "[2] Phase"=$Particles.Age
      "[3] Phase"=$Emitter.Age
     ))) {enabled=false, isSetParameters=true, module=SetVariables_#4}
   /SetVariables(Particles.FlapTarget=1) {isSetParameters=true, module=SetVariables_#5}
   &InterpolateOverTime(
    Rate Of Change:NiagaraFloat=&Lerp_Float(A=1, B=1, Alpha=$Particles.FlapTarget)
    Target Float Value:NiagaraFloat=$Particles.FlapTarget
   )
   /SetVariables(Particles.FlapSmoothed:NiagaraFloat=$Particles.InterpolateOverTime.InterpolateOverTime_Float.Moving Average) {isSetParameters=true, module=SetVariables_#6}
   &IncrementOverTime(Rate of Change:NiagaraFloat=&Multiply_Float(
     A=&Add_Float(A=0.15, B=$Particles.FlapSmoothed)
     B=&RemapRange(
      Input Value To Remap=&VectorLength(Vector=$Particles.Velocity)
      Input Min=0
      Input Max=5000
      Output Min=1
      Output Max=5
     )
    ))
   &DynamicMaterialParameters(
    Index 0 Param 1=&Multiply_Float(A=$Particles.FlapSmoothed, B=1)
    Index 0 Param 2=&Multiply_Float(A=$Particles.IncrementOverTime.Counter, B=1)
    Module.Param0WriteEnabled=true
    Module.Param1WriteEnabled=true
    Module.Param2WriteEnabled=false
    Module.Param3WriteEnabled=false
   ) {module=DynamicMaterialParameters001}
   &RegionMask(Outer Vector:Vector3f=<0 0 2>, Region Radius=1200, UseOffset=false, UseRotation=false, Value Type:ENiagaraRegionMaskValue=::NewEnumerator1)
   /SetVariables(Particles.Scale=$Particles.RegionMask.Vector) {isSetParameters=true, module=SetVariables_#7}
   &KillParticlesInVolume(Invert Volume=true, Sphere Radius=2500)
   /SetVariables(Particles.NewScale:Vector3f=&Multiply_VectorByFloat(
     Vector=$Particles.Scale
     Float=&FloatFromCurve(
      FloatCurve=DataInterface() {propertyValues {Curve {keys [
          {interpMode=RCIM_Cubic, tangentMode=RCTM_User, time=0.1958, value=0.1176}
          {arriveTangent=2.4807, interpMode=RCIM_Cubic, leaveTangent=2.4807, tangentMode=RCTM_User, time=0.3694, value=0.9922}
          {interpMode=RCIM_Cubic, tangentMode=RCTM_User, time=1.0001, value=1.7529}
         ]}}}
      Scale Curve=0.15
     )
    )) {isSetParameters=true, module=SetVariables_#8}
  ]
  rendererClasses [&NiagaraMeshRendererProperties, &NiagaraSpriteRendererProperties]
  renderers [
   {rendererClass=&NiagaraMeshRendererProperties, rendererIndex=0}
   {rendererClass=&NiagaraSpriteRendererProperties, rendererIndex=1}
  ]
 }
 GroundSnow {
  emitterUpdateScript [
   &EmitterState()
   &SpawnRate(SpawnRate=3, Use Spawn Probability=false)
  ]
  particleSpawnScript [
   &InitializeParticle(
    Color=<rgba 0.8038 0.5329 0.3235 0.4>
    Lifetime=5
    Mesh Scale Mode=::NewEnumerator3
    Mesh Uniform Scale=80
    UsePositionOffset=false
    Write Material Random=true
   ) {module=InitializeParticle001}
   &ShapeLocation(
    Box Size=<300 1000 75>
    MaskSpawnGroup=false
    Offset=<200 0 -75>
    Offset Mode=::NewEnumerator0
    Override Seed=false
    Shape Primitive=::NewEnumerator2
    Surface Only=false
    UseOwnerScale=true
   )
   &DynamicMaterialParameters(
    Index 0 Param 1=&RandomRangeFloat(Minimum=0, Maximum=250)
    Index 0 Param 2=&RandomRangeFloat(Minimum=0, Maximum=0.4)
    Index 0 Param 3=&RandomRangeFloat(Minimum=0, Maximum=1)
    Index 0 Param 4=1
    Module.Index0Float0Write=true
    Module.Index0Float1Write=true
    Module.Index0Float2Write=true
    Module.Index0Float3Write=true
   )
  ]
  particleUpdateScript [
   &ParticleState()
   &SolveForcesAndVelocity(Clamp Velocity=false, Limit Acceleration=false)
   &ScaleColor(
    Scale Alpha=&FloatFromCurve(FloatCurve=DataInterface() {propertyValues {Curve {
        defaultValue=1
        keys [
         {interpMode=RCIM_Cubic, tangentMode=RCTM_User, time=0, value=0}
         {interpMode=RCIM_Cubic, tangentMode=RCTM_User, time=0.5, value=1}
         {tangentMode=RCTM_User, time=1, value=0}
        ]
       }}})
    ScaleA=true
    ScaleRGB=true
   )
  ]
  rendererClasses [&NiagaraMeshRendererProperties]
  renderers [{rendererClass=&NiagaraMeshRendererProperties, rendererIndex=0}]
 }
 LensDust {
  emitterUpdateScript [
   &EmitterState(Life Cycle Mode=::NewEnumerator1, Loop Behavior=::NewEnumerator1, Loop Duration Mode=::NewEnumerator1, UseLoopDelay=false)
   &SpawnBurst_Instantaneous(Spawn Count=15, Use Spawn Probability=false, UseLoopCountLimit=false)
  ]
  particleSpawnScript [&InitializeParticle(
    AdjustAlpha=true
    AdjustHue=true
    AdjustSaturation=false
    AdjustValue=false
    Alpha Scale Range:Vector2f=<0.05 0.1>
    Color=<rgba 1 0.3735 0.0059 1>
    Color Mode=::NewEnumerator3
    Lifetime=5
    Position Offset=&MakeVector(
     X=&Lerp_Float(A=550, B=720, Alpha=&ReturnNormalizedExecIndex())
     Y=0
     Z=&Lerp_Float(A=40, B=220, Alpha=&ReturnNormalizedExecIndex())
    )
    Sprite Size Mode=::NewEnumerator1
    Uniform Sprite Size Max=12
    Uniform Sprite Size Min=150
    UsePositionOffset=true
    Write Material Random=true
   )]
  particleUpdateScript [
   &ParticleState(Kill Particles When Lifetime Has Elapsed=false)
   &ScaleColor(Scale Alpha=&Lerp_Float(A=0.4, B=1, Alpha=$System.Timeline.Time), ScaleA=true, ScaleRGB=true)
  ]
  rendererClasses [&NiagaraSpriteRendererProperties]
  renderers [{rendererClass=&NiagaraSpriteRendererProperties}]
 }
 LensFlare {
  emitterUpdateScript [
   &EmitterState(Life Cycle Mode=::NewEnumerator1, Loop Behavior=::NewEnumerator1, Loop Duration Mode=::NewEnumerator1, UseLoopDelay=false)
   &SpawnBurst_Instantaneous(Use Spawn Probability=false, UseLoopCountLimit=false)
  ]
  particleSpawnScript [&InitializeParticle(
    Color=<rgba 2 0.2003 0.0007 1>
    Lifetime=2
    Position Offset=<705 0 220>
    Sprite Size Mode=::NewEnumerator3
    Uniform Sprite Size=400
    UsePositionOffset=true
    Write Material Random=true
   )]
  particleUpdateScript [
   &ParticleState(Kill Particles When Lifetime Has Elapsed=false, Loop Particles Lifetime=true)
   &ScaleColor(ScaleA=true, ScaleRGB=true)
   /SetVariables(
    Particles.CoronaColor:LinearColor=&LerpLinearColors(StartColor=<rgba 1 1 1 0.1>, EndColor=<rgba 1 1 1 0.25>, LerpFactor=$System.Timeline.Time)
    Particles.CoronaSize:Vector2f=&Vector2DFromFloat(Value=440)
   ) {isSetParameters=true}
  ]
  rendererClasses [&NiagaraSpriteRendererProperties]
  renderers [
   {rendererClass=&NiagaraSpriteRendererProperties}
   {rendererClass=&NiagaraSpriteRendererProperties, rendererIndex=1}
  ]
 }
 LightGlow {
  emitterUpdateScript [
   &EmitterState(Life Cycle Mode=::NewEnumerator1, Loop Behavior=::NewEnumerator1, Loop Duration Mode=::NewEnumerator1, UseLoopDelay=false)
   &SpawnBurst_Instantaneous(Spawn Count=4, Use Spawn Probability=false, UseLoopCountLimit=false)
  ]
  particleSpawnScript [&InitializeParticle(
    Color=&SelectLinearColorFromArray(
     Array Sampling Mode=::NewEnumerator1
     Direct Array Index=&ReturnExecIndex()
     Color Selection Array=DataInterface() {propertyValues {ColorData [<rgba 1 0.5378 0 0>, <rgba 1 0.5378 0 0.4>, <rgba 1 0.5378 0 0>, <rgba 0.0089 0.6528 1 0.5>]}}
     Random Seed=?
    )
    Position Offset=&SelectVectorFromArray(
     Direct Array Index=&ReturnExecIndex()
     Vector Selection Array=DataInterface() {propertyValues {FloatData [<-304 0 65>, <-696 0 100>, <398 0 55>, <-157 0 -16>]}}
    )
    Sprite Size Mode=::NewEnumerator3
    Uniform Sprite Size=&SelectFloatFromArray(
     Direct Array Index=&ReturnExecIndex()
     Float Selection Array=DataInterface() {propertyValues {FloatData [35,35,35,20]}}
    )
    UsePositionOffset=true
    Write Material Random=true
   )]
  particleUpdateScript [
   &ParticleState(Kill Particles When Lifetime Has Elapsed=false)
   &ScaleColor(
    Scale Alpha=&RandomRangeFloat(Minimum=0.74, Maximum=1, Evaluation Type=::NewEnumerator1)
    ScaleA=true
    ScaleRGB=true
   )
   &OffsetPosition(Position Offset=&MakeCustomVectorFromBool(
     Bool=$System.IsZoomedOut
     True Vector=<0 0 0>
     False Vector=&SelectVectorFromArray(
      Direct Array Index=&ReturnExecIndex()
      Vector Selection Array=DataInterface() {propertyValues {FloatData [<4 0 0>, <10 0 1>, <-3 0 2>, <0 0 0>]}}
     )
    ))
  ]
  rendererClasses [&NiagaraSpriteRendererProperties]
  renderers [{rendererClass=&NiagaraSpriteRendererProperties}]
 }
}
summary {
 systemName=NS_4100_Frontend
 emitters [
  {emitterName=GroundSnow, rendererClasses [&NiagaraMeshRendererProperties]}
  {emitterName=LightGlow, rendererClasses [&NiagaraSpriteRendererProperties]}
  {emitterName=AirLines, rendererClasses [&NiagaraMeshRendererProperties]}
  {emitterName=AirLines-InFront, rendererClasses [&NiagaraMeshRendererProperties]}
  {
   emitterName=Birds
   simTarget=GPUComputeSim
   rendererClasses [&NiagaraMeshRendererProperties, &NiagaraSpriteRendererProperties]
  }
  {
   emitterName=Birds001
   simTarget=GPUComputeSim
   rendererClasses [&NiagaraMeshRendererProperties, &NiagaraSpriteRendererProperties]
  }
  {emitterName=LensFlare, rendererClasses [&NiagaraSpriteRendererProperties]}
  {emitterName=LensDust, rendererClasses [&NiagaraSpriteRendererProperties]}
  {emitterName=BigGlow, rendererClasses [&NiagaraSpriteRendererProperties]}
 ]
 userVariables [
  User.Landscape {
   type=NiagaraDataInterfaceLandscape
   description=""
   defaultValue {
    struct=NiagaraExt_VariableValue_DataInterface
    value {
     dataInterface=&self:NiagaraDataInterfaceLandscape_0
     dataInterfaceClass=&NiagaraDataInterfaceLandscape
    }
   }
  }
  User.RadiusMult {type=NiagaraFloat, defaultValue=1, description=""}
  User.SpawnRadius {type=NiagaraFloat, defaultValue=2000, description=""}
  User.SpawnRate {type=NiagaraFloat, defaultValue=1, description=""}
 ]
}
